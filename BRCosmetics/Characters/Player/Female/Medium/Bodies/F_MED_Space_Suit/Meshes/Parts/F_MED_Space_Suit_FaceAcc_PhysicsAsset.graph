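# Physics asset for the face accessory part of the female medium space suit character cosmetic.
# It holds a single body on the head bone, shaped as one rotated, offset capsule.
# It defines no constraints between bodies.
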PhysicsAsset /BRCosmetics/Characters/Player/Female/Medium/Bodies/F_MED_Space_Suit/Meshes/Parts/F_MED_Space_Suit_FaceAcc_PhysicsAsset
bodies {head [{shapeType=Capsule, length=11.4162, radius=9.878, center=<4.1867 2.3495 -0.3962>, rotation=<rot 0 52.8477 -59.939>}]}
constraints []
